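# Physics asset for the back skirt of the SprocketPoppy BoatCircus medium male body, providing capsule bodies for skirt, pelvis and thighs.
# Left and right skirt chains hang from master bones on the pelvis, each running 01-02-03 with 04 branching off 01.
# Chain joints allow wide swing on one axis with tight second-swing and twist limits.
# Thigh and master joints are angularly free, as is a cross-constraint tying the two 02 bones together.
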PhysicsAsset /BRCosmetics/Characters/Player/Male/Medium/Bodies/M_MED_SprocketPoppy_BoatCircus/Meshes/M_MED_SprocketPoppy_BoatCircus_BackSkirt_Physics
bodies {
 dyn_bach_skirt_l_01 [{shapeType=Capsule, length=5, radius=2.5, shapeName=dyn_bach_skirt_l_01_capsule, center=<4 0 0>, rotation=<rot 90 0 0>}]
 dyn_bach_skirt_l_02 [{shapeType=Capsule, length=5, radius=2.5, shapeName=dyn_bach_skirt_l_02_capsule, center=<4 0 0>, rotation=<rot 90 0 0>}]
 dyn_bach_skirt_l_03 [{shapeType=Capsule, length=7.5, radius=3.5, shapeName=dyn_bach_skirt_l_03_capsule, center=<5 0 -1>, rotation=<rot 40 0 0>}]
 dyn_bach_skirt_l_04 [{shapeType=Capsule, length=6.5, radius=2.5, shapeName=dyn_bach_skirt_l_04_capsule, center=<4 0 2>, rotation=<rot 90 0 0>}]
 dyn_bach_skirt_l_mstr [{shapeType=Capsule, length=0.505, radius=0.505, shapeName=dyn_bach_skirt_l_mstr_capsule, rotation=<rot 0 -90 0>}]
 dyn_bach_skirt_r_01 [{shapeType=Capsule, length=5, radius=2.5, shapeName=dyn_bach_skirt_r_01_capsule, center=<4 0 0>, rotation=<rot 90 0 0>}]
 dyn_bach_skirt_r_02 [{shapeType=Capsule, length=5, radius=2.5, shapeName=dyn_bach_skirt_r_02_capsule, center=<4 0 0>, rotation=<rot 90 0 0>}]
 dyn_bach_skirt_r_03 [{shapeType=Capsule, length=7.5, radius=3.5, shapeName=dyn_bach_skirt_r_03_capsule, center=<5 0 1>, rotation=<rot -40 0 0>}]
 dyn_bach_skirt_r_04 [{shapeType=Capsule, length=6.5, radius=2.5, shapeName=dyn_bach_skirt_r_04_capsule, center=<4 0 -2>, rotation=<rot 90 0 0>}]
 dyn_bach_skirt_r_mstr [{shapeType=Capsule, length=0.505, radius=0.505, shapeName=dyn_bach_skirt_r_mstr_capsule, rotation=<rot 0 -90 0>}]
 pelvis [{
   shapeType=Capsule
   length=17.2736
   radius=7.9175
   shapeName=pelvis_capsule
   center=<-4.9874 -2.6803 0.0392>
   rotation=<rot -0.0783 -90 0.731>
  }]
 thigh_l [{shapeType=Capsule, length=30, radius=9.5, shapeName=thigh_l_capsule, center=<-18 -1.9 0>, rotation=<rot 90 180 177>}]
 thigh_r [{shapeType=Capsule, length=30, radius=9.5, shapeName=thigh_r_capsule, center=<18 1.9 0>, rotation=<rot -90 180 -177>}]
}
constraints [
 {
  bone1Name=thigh_l
  bone2Name=pelvis
  swing1LimitDegrees=45
  swing1Motion=Free
  swing2LimitDegrees=45
  swing2Motion=Free
  twistLimitDegrees=45
  twistMotion=Free
 }
 {
  bone1Name=thigh_r
  bone2Name=pelvis
  swing1LimitDegrees=45
  swing1Motion=Free
  swing2LimitDegrees=45
  swing2Motion=Free
  twistLimitDegrees=45
  twistMotion=Free
 }
 {
  bone1Name=dyn_bach_skirt_l_mstr
  bone2Name=pelvis
  swing1LimitDegrees=45
  swing1Motion=Free
  swing2LimitDegrees=45
  swing2Motion=Free
  twistLimitDegrees=45
  twistMotion=Free
 }
 {bone1Name=dyn_bach_skirt_l_01, bone2Name=dyn_bach_skirt_l_mstr, swing1LimitDegrees=45, swing2LimitDegrees=1, twistLimitDegrees=1}
 {bone1Name=dyn_bach_skirt_l_02, bone2Name=dyn_bach_skirt_l_01, swing1LimitDegrees=45, swing2LimitDegrees=1, twistLimitDegrees=1}
 {bone1Name=dyn_bach_skirt_l_03, bone2Name=dyn_bach_skirt_l_02, swing1LimitDegrees=45, swing2LimitDegrees=5, twistLimitDegrees=5}
 {bone1Name=dyn_bach_skirt_l_04, bone2Name=dyn_bach_skirt_l_01, swing1LimitDegrees=45, swing2LimitDegrees=10, twistLimitDegrees=1}
 {
  bone1Name=dyn_bach_skirt_r_mstr
  bone2Name=pelvis
  swing1LimitDegrees=45
  swing1Motion=Free
  swing2LimitDegrees=45
  swing2Motion=Free
  twistLimitDegrees=45
  twistMotion=Free
 }
 {bone1Name=dyn_bach_skirt_r_01, bone2Name=dyn_bach_skirt_r_mstr, swing1LimitDegrees=45, swing2LimitDegrees=1, twistLimitDegrees=1}
 {bone1Name=dyn_bach_skirt_r_02, bone2Name=dyn_bach_skirt_r_01, swing1LimitDegrees=60, swing2LimitDegrees=1, twistLimitDegrees=1}
 {bone1Name=dyn_bach_skirt_r_03, bone2Name=dyn_bach_skirt_r_02, swing1LimitDegrees=45, swing2LimitDegrees=5, twistLimitDegrees=5}
 {bone1Name=dyn_bach_skirt_r_04, bone2Name=dyn_bach_skirt_r_01, swing1LimitDegrees=45, swing2LimitDegrees=10, twistLimitDegrees=1}
 {
  bone1Name=dyn_bach_skirt_l_02
  bone2Name=dyn_bach_skirt_r_02
  swing1LimitDegrees=60
  swing1Motion=Free
  swing2LimitDegrees=1
  swing2Motion=Free
  twistLimitDegrees=1
  twistMotion=Free
 }
]
