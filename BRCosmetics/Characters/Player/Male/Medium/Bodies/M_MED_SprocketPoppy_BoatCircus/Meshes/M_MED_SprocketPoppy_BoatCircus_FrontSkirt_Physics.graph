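# Physics asset for the front skirt of the male medium SprocketPoppy BoatCircus body, setting up its dynamic cloth-like bone motion.
# Center, left and right chains of four capsules each hang from tiny master bodies constrained to the pelvis.
# Chain joints bend mainly on one swing axis with almost no twist or sideways swing, while angular-free cross constraints tie the lower two side segments to the center chain.
# Pelvis and thigh capsules supply body volumes.
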PhysicsAsset /BRCosmetics/Characters/Player/Male/Medium/Bodies/M_MED_SprocketPoppy_BoatCircus/Meshes/M_MED_SprocketPoppy_BoatCircus_FrontSkirt_Physics
bodies {
 dyn_front_skirt_c_01 [{shapeType=Capsule, length=3.5, radius=4.5, shapeName=dyn_front_skirt_c_01_capsule, center=<6 1 0>, rotation=<rot 90 0 0>}]
 dyn_front_skirt_c_02 [{shapeType=Capsule, length=3.5, radius=4.5, shapeName=dyn_front_skirt_c_02_capsule, center=<6 1 0>, rotation=<rot 90 0 0>}]
 dyn_front_skirt_c_03 [{shapeType=Capsule, length=3, radius=5, shapeName=dyn_front_skirt_c_03_capsule, center=<6 1 0>, rotation=<rot 90 0 0>}]
 dyn_front_skirt_c_04 [{shapeType=Capsule, length=3, radius=5, shapeName=dyn_front_skirt_c_04_capsule, center=<6 1 0>, rotation=<rot 90 0 0>}]
 dyn_front_skirt_c_mstr [{shapeType=Capsule, length=0.505, radius=0.505, shapeName=dyn_front_skirt_c_mstr_capsule, rotation=<rot 0 -90 0>}]
 dyn_front_skirt_l_01 [{shapeType=Capsule, length=4, radius=3, shapeName=dyn_front_skirt_l_01_capsule, center=<4.3297 1.2 -2>, rotation=<rot 90 0 0>}]
 dyn_front_skirt_l_02 [{shapeType=Capsule, length=4, radius=3, shapeName=dyn_front_skirt_l_02_capsule, center=<4.3297 1.2 -2>, rotation=<rot 90 0 0>}]
 dyn_front_skirt_l_03 [{shapeType=Capsule, length=3.5, radius=3.5, shapeName=dyn_front_skirt_l_03_capsule, center=<4.3297 1.2 -2>, rotation=<rot 90 0 0>}]
 dyn_front_skirt_l_04 [{shapeType=Capsule, length=3.5, radius=3.5, shapeName=dyn_front_skirt_l_04_capsule, center=<4.3297 1.2 -2>, rotation=<rot 90 0 0>}]
 dyn_front_skirt_l_mstr [{shapeType=Capsule, length=0.505, radius=0.505, shapeName=dyn_front_skirt_l_mstr_capsule, rotation=<rot 0 -90 0>}]
 dyn_front_skirt_r_01 [{shapeType=Capsule, length=4, radius=3, shapeName=dyn_front_skirt_r_01_capsule, center=<4.5317 1.2 2>, rotation=<rot 90 0 0>}]
 dyn_front_skirt_r_02 [{shapeType=Capsule, length=4, radius=3, shapeName=dyn_front_skirt_r_02_capsule, center=<4.5317 1.2 2>, rotation=<rot 90 0 0>}]
 dyn_front_skirt_r_03 [{shapeType=Capsule, length=3.5, radius=3.5, shapeName=dyn_front_skirt_r_03_capsule, center=<4.5317 1.2 2>, rotation=<rot 90 0 0>}]
 dyn_front_skirt_r_04 [{shapeType=Capsule, length=3.5, radius=3.5, shapeName=dyn_front_skirt_r_04_capsule, center=<4.5317 1.2 2>, rotation=<rot 90 0 0>}]
 dyn_front_skirt_r_mstr [{shapeType=Capsule, length=0.505, radius=0.505, shapeName=dyn_front_skirt_r_mstr_capsule, rotation=<rot 0 -90 0>}]
 pelvis [{
   shapeType=Capsule
   length=13.2736
   radius=9.9175
   shapeName=pelvis_capsule
   center=<-5.3939 -2.6429 0.0444>
   rotation=<rot -0.0783 -90 0.731>
  }]
 thigh_l [{shapeType=Capsule, length=29.5, radius=10, shapeName=thigh_l_capsule, center=<-17.8513 0.951 0.4698>, rotation=<rot 90 89.6476 100.1578>}]
 thigh_r [{shapeType=Capsule, length=30, radius=9.5, shapeName=thigh_r_capsule, center=<18 -0.26 0>, rotation=<rot -90 90 -97>}]
}
constraints [
 {bone1Name=dyn_front_skirt_c_01, bone2Name=dyn_front_skirt_c_mstr, swing1LimitDegrees=50, swing2LimitDegrees=0, twistLimitDegrees=0}
 {bone1Name=dyn_front_skirt_c_02, bone2Name=dyn_front_skirt_c_01, swing1LimitDegrees=45, swing2LimitDegrees=0, twistLimitDegrees=0}
 {bone1Name=dyn_front_skirt_c_03, bone2Name=dyn_front_skirt_c_02, swing1LimitDegrees=45, swing2LimitDegrees=10, twistLimitDegrees=0}
 {bone1Name=dyn_front_skirt_c_04, bone2Name=dyn_front_skirt_c_03, swing1LimitDegrees=45, swing2LimitDegrees=0, twistLimitDegrees=0}
 {bone1Name=dyn_front_skirt_l_01, bone2Name=dyn_front_skirt_l_mstr, swing1LimitDegrees=50, swing2LimitDegrees=0, twistLimitDegrees=0}
 {bone1Name=dyn_front_skirt_l_02, bone2Name=dyn_front_skirt_l_01, swing1LimitDegrees=45, swing2LimitDegrees=0, twistLimitDegrees=0}
 {bone1Name=dyn_front_skirt_l_03, bone2Name=dyn_front_skirt_l_02, swing1LimitDegrees=45, swing2LimitDegrees=0, twistLimitDegrees=0}
 {bone1Name=dyn_front_skirt_l_04, bone2Name=dyn_front_skirt_l_03, swing1LimitDegrees=45, swing2LimitDegrees=0, twistLimitDegrees=0}
 {bone1Name=dyn_front_skirt_r_01, bone2Name=dyn_front_skirt_r_mstr, swing1LimitDegrees=50, swing2LimitDegrees=0, twistLimitDegrees=0}
 {bone1Name=dyn_front_skirt_r_02, bone2Name=dyn_front_skirt_r_01, swing1LimitDegrees=45, swing2LimitDegrees=0, twistLimitDegrees=0}
 {bone1Name=dyn_front_skirt_r_03, bone2Name=dyn_front_skirt_r_02, swing1LimitDegrees=45, swing2LimitDegrees=0, twistLimitDegrees=0}
 {bone1Name=dyn_front_skirt_r_04, bone2Name=dyn_front_skirt_r_03, swing1LimitDegrees=45, swing2LimitDegrees=0, twistLimitDegrees=0}
 {
  bone1Name=dyn_front_skirt_l_mstr
  bone2Name=pelvis
  swing1LimitDegrees=45
  swing1Motion=Free
  swing2LimitDegrees=45
  swing2Motion=Free
  twistLimitDegrees=45
  twistMotion=Free
 }
 {
  bone1Name=dyn_front_skirt_c_mstr
  bone2Name=pelvis
  swing1LimitDegrees=45
  swing1Motion=Free
  swing2LimitDegrees=45
  swing2Motion=Free
  twistLimitDegrees=45
  twistMotion=Free
 }
 {
  bone1Name=dyn_front_skirt_r_mstr
  bone2Name=pelvis
  swing1LimitDegrees=45
  swing1Motion=Free
  swing2LimitDegrees=45
  swing2Motion=Free
  twistLimitDegrees=45
  twistMotion=Free
 }
 {
  bone1Name=thigh_l
  bone2Name=pelvis
  swing1LimitDegrees=45
  swing1Motion=Free
  swing2LimitDegrees=45
  swing2Motion=Free
  twistLimitDegrees=45
  twistMotion=Free
 }
 {
  bone1Name=thigh_r
  bone2Name=pelvis
  swing1LimitDegrees=45
  swing1Motion=Free
  swing2LimitDegrees=45
  swing2Motion=Free
  twistLimitDegrees=45
  twistMotion=Free
 }
 {
  bone1Name=dyn_front_skirt_r_03
  bone2Name=dyn_front_skirt_c_03
  swing1LimitDegrees=45
  swing1Motion=Free
  swing2LimitDegrees=45
  swing2Motion=Free
  twistLimitDegrees=45
  twistMotion=Free
 }
 {
  bone1Name=dyn_front_skirt_l_03
  bone2Name=dyn_front_skirt_c_03
  swing1LimitDegrees=45
  swing1Motion=Free
  swing2LimitDegrees=45
  swing2Motion=Free
  twistLimitDegrees=45
  twistMotion=Free
 }
 {
  bone1Name=dyn_front_skirt_l_04
  bone2Name=dyn_front_skirt_c_04
  swing1LimitDegrees=45
  swing1Motion=Free
  swing2LimitDegrees=45
  swing2Motion=Free
  twistLimitDegrees=45
  twistMotion=Free
 }
 {
  bone1Name=dyn_front_skirt_r_04
  bone2Name=dyn_front_skirt_c_04
  swing1LimitDegrees=45
  swing1Motion=Free
  swing2LimitDegrees=45
  swing2Motion=Free
  twistLimitDegrees=45
  twistMotion=Free
 }
]
